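# Fortnite GigaMix-based environment material that adds a custom animated, optionally reactive emissive.
# GigaMix attributes pass through shared damage, wind, snow, Veneer, RVT and jungle functions, with a cheaper low-quality mobile branch.
# A texture mask, hue rotation and sine pulse drive the glow, which is compensated for eye adaptation.
# Custom primitive data timestamps feed fade in/out timelines for the reactive boost.
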Material /Game/Environments/Helios/TechArt/ParkEmissive/P_FN_Bas_CustomEmissive
graph {
 expressions {
  MaterialExpressionAbs_1 {class=MaterialExpressionAbs, inputs [None {from=MaterialExpressionAdd_53, output=""}], properties {}}
  MaterialExpressionAbs_5 {class=MaterialExpressionAbs, inputs [None {from=MaterialExpressionAdd_54, output=""}], properties {}}
  MaterialExpressionAbs_6 {class=MaterialExpressionAbs, inputs [None {from=MaterialExpressionAdd_55, output=""}], properties {}}
  MaterialExpressionAbs_7 {class=MaterialExpressionAbs, inputs [None {from=MaterialExpressionAdd_56, output=""}], properties {}}
  MaterialExpressionAdd_0 {
   class=MaterialExpressionAdd
   inputs [
    A {from=MaterialExpressionScalarParameter_10}
    B {from=MaterialExpressionMultiply_17}
   ]
  }
  MaterialExpressionAdd_1 {
   class=MaterialExpressionAdd
   inputs [
    A {from=MaterialExpressionMultiply_6}
    B {from=MaterialExpressionMultiply_13}
   ]
  }
  MaterialExpressionAdd_2 {
   class=MaterialExpressionAdd
   inputs [
    A {from=MaterialExpressionDivide_1}
    B {from=MaterialExpressionDesaturation_1}
   ]
  }
  MaterialExpressionAdd_3 {
   class=MaterialExpressionAdd
   inputs [
    A {from=MaterialExpressionSine_0}
    B {from=""}
   ]
  }
  MaterialExpressionAdd_4 {
   class=MaterialExpressionAdd
   inputs [
    A {from=MaterialExpressionComponentMask_0}
    B {from=MaterialExpressionComponentMask_1}
   ]
  }
  MaterialExpressionAdd_5 {
   class=MaterialExpressionAdd
   inputs [
    A {from=MaterialExpressionMaterialFunctionCall_13, output=Rotated Values}
    B {from=MaterialExpressionConstant2Vector_1}
   ]
  }
  MaterialExpressionAdd_53 {
   class=MaterialExpressionAdd
   inputs [
    A {from=MaterialExpressionComponentMask_8}
    B {from=""}
   ]
   properties {constB=-0.5}
  }
  MaterialExpressionAdd_54 {
   class=MaterialExpressionAdd
   inputs [
    A {from=MaterialExpressionAdd_53}
    B {from=""}
   ]
   properties {constB=-1}
  }
  MaterialExpressionAdd_55 {
   class=MaterialExpressionAdd
   inputs [
    A {from=MaterialExpressionAdd_54}
    B {from=""}
   ]
   properties {constB=-1}
  }
  MaterialExpressionAdd_56 {
   class=MaterialExpressionAdd
   inputs [
    A {from=MaterialExpressionAdd_55}
    B {from=""}
   ]
   properties {constB=-1}
  }
  MaterialExpressionAdd_7 {
   class=MaterialExpressionAdd
   inputs [
    A {from=MaterialExpressionMultiply_11}
    B {from=MaterialExpressionScalarParameter_18}
   ]
  }
  MaterialExpressionCeil_0 {class=MaterialExpressionCeil, inputs [None {from=MaterialExpressionScalarParameter_14, output=""}], properties {}}
  MaterialExpressionClamp_0 {
   class=MaterialExpressionClamp
   inputs [
    None {from=MaterialExpressionOneMinus_0, output=""}
    Min {from="", output=""}
    Max {from="", output=""}
   ]
   properties {minDefault=0}
  }
  MaterialExpressionComponentMask_0 {class=MaterialExpressionComponentMask, inputs [None {from=MaterialExpressionNamedRerouteUsage_3}], properties {r=true}}
  MaterialExpressionComponentMask_1 {class=MaterialExpressionComponentMask, inputs [None {from=MaterialExpressionNamedRerouteUsage_3}], properties {b=true}}
  MaterialExpressionComponentMask_8 {class=MaterialExpressionComponentMask, inputs [None {from=MaterialExpressionTextureCoordinate_5}], properties {r=true}}
  MaterialExpressionComposite_0 {
   class=MaterialExpressionComposite
   inputs [None {from=MaterialExpressionReroute_68, output=Result}]
   properties {desc=Transform to World in case we will use RVT after this point, subgraphName=Tangent to World}
  }
  MaterialExpressionComposite_1 {
   class=MaterialExpressionComposite
   inputs [MaterialAttributes {from=MaterialExpressionMaterialFunctionCall_16, output=Result}]
   properties {subgraphName=Documentation Link}
  }
  MaterialExpressionComposite_2 {
   class=MaterialExpressionComposite
   inputs [MaterialAttributes {from=MaterialExpressionReroute_71, output=Output}]
   properties {desc=Game Shared WPO and Fire Effect, subgraphName=WPO And Shared Functions}
  }
  MaterialExpressionComposite_4 {
   class=MaterialExpressionComposite
   inputs [InputPin {from=MaterialExpressionMaterialFunctionCall_17, output=Result}]
   properties {subgraphName=Mobile Opt}
  }
  MaterialExpressionComposite_7 {
   class=MaterialExpressionComposite
   inputs [
    "For Moss (MA)" {from=MaterialExpressionMaterialFunctionCall_4, output=Result}
    "RVT_Height (S)" {from=MaterialExpressionMaterialFunctionCall_17, output=RVT H}
   ]
   properties {subgraphName="Seasonal "}
  }
  MaterialExpressionConstant2Vector_0 {class=MaterialExpressionConstant2Vector, properties {g=0.3, r=0.2}}
  MaterialExpressionConstant2Vector_1 {class=MaterialExpressionConstant2Vector, properties {g=0.25, r=0.1}}
  MaterialExpressionConstant3Vector_0 {class=MaterialExpressionConstant3Vector, properties {constant=<rgba 1 1 1 0>}}
  MaterialExpressionConstant_0 {class=MaterialExpressionConstant, properties {r=1}}
  MaterialExpressionConstant_1 {class=MaterialExpressionConstant, properties {r=3}}
  MaterialExpressionConstant_2 {class=MaterialExpressionConstant, properties {r=0.2858}}
  MaterialExpressionConstant_3 {class=MaterialExpressionConstant, properties {r=0.25}}
  MaterialExpressionConstant_4 {class=MaterialExpressionConstant, properties {r=0.3}}
  MaterialExpressionConstant_6 {class=MaterialExpressionConstant, properties {r=0.8}}
  MaterialExpressionConstant_7 {class=MaterialExpressionConstant, properties {r=0.5}}
  MaterialExpressionConvert_0 {
   class=MaterialExpressionConvert
   inputs [None {from=MaterialExpressionNamedRerouteUsage_0}]
   properties {
    nodeName=Break Float3
    convertInputs [{
      type=Vector3
      expressionInput {
       expression=&self:MaterialExpressionNamedRerouteUsage_0
       inputName=None
       mask=0
       maskA=0
       maskB=0
       maskG=0
       maskR=0
       outputIndex=0
      }
     }]
    convertOutputs [{type=Scalar}, {type=Scalar}, {type=Scalar}]
   }
  }
  MaterialExpressionConvert_1 {
   class=MaterialExpressionConvert
   inputs [
    None {from=MaterialExpressionMultiply_8}
    None {from=MaterialExpressionMultiply_9}
    None {from=MaterialExpressionConvert_0}
   ]
   properties {
    nodeName=Make Vector3
    convertInputs [
     {
      type=Scalar
      expressionInput {
       expression=&self:MaterialExpressionMultiply_8
       inputName=None
       mask=0
       maskA=0
       maskB=0
       maskG=0
       maskR=0
       outputIndex=0
      }
     }
     {
      type=Scalar
      expressionInput {
       expression=&self:MaterialExpressionMultiply_9
       inputName=None
       mask=0
       maskA=0
       maskB=0
       maskG=0
       maskR=0
       outputIndex=0
      }
     }
     {
      type=Scalar
      expressionInput {
       expression=&self:MaterialExpressionConvert_0
       inputName=None
       mask=0
       maskA=0
       maskB=0
       maskG=0
       maskR=0
       outputIndex=2
      }
     }
    ]
    convertOutputs [{type=Vector3}]
   }
  }
  MaterialExpressionDesaturation_1 {
   class=MaterialExpressionDesaturation
   inputs [
    None {from=MaterialExpressionLinearInterpolate_1, output=""}
    Fraction {from="", output=""}
   ]
   properties {}
  }
  MaterialExpressionDivide_0 {
   class=MaterialExpressionDivide
   inputs [
    A {from=MaterialExpressionTime_1}
    B {from=MaterialExpressionScalarParameter_17}
   ]
   properties {constB=4}
  }
  MaterialExpressionDivide_1 {
   class=MaterialExpressionDivide
   inputs [
    A {from=MaterialExpressionTime_0}
    B {from=MaterialExpressionScalarParameter_5}
   ]
  }
  MaterialExpressionDivide_2 {
   class=MaterialExpressionDivide
   inputs [
    A {from=MaterialExpressionSubtract_1}
    B {from=MaterialExpressionMultiply_14}
   ]
  }
  MaterialExpressionDivide_3 {
   class=MaterialExpressionDivide
   inputs [
    A {from=MaterialExpressionSubtract_2}
    B {from=MaterialExpressionScalarParameter_11}
   ]
  }
  MaterialExpressionDivide_4 {
   class=MaterialExpressionDivide
   inputs [
    A {from=MaterialExpressionSubtract_3}
    B {from=MaterialExpressionScalarParameter_16}
   ]
  }
  MaterialExpressionDotProduct_2 {
   class=MaterialExpressionDotProduct
   inputs [
    A {from=MaterialExpressionConstant3Vector_0, output=""}
    B {from=MaterialExpressionNamedRerouteDeclaration_0, output=""}
   ]
   properties {}
  }
  MaterialExpressionEyeAdaptationInverse_0 {
   class=MaterialExpressionEyeAdaptationInverse
   inputs [
    LightValueInput {from=MaterialExpressionStaticSwitchParameter_1, output=""}
    AlphaInput {from=MaterialExpressionConstant_4, output=""}
   ]
   properties {}
  }
  MaterialExpressionGetMaterialAttributes_0 {
   class=MaterialExpressionGetMaterialAttributes
   inputs [None {from=MaterialExpressionReroute_3, output=Input}]
   properties {attributeGetTypes [#1]}
  }
  MaterialExpressionGetMaterialAttributes_1 {
   class=MaterialExpressionGetMaterialAttributes
   inputs [None {from=MaterialExpressionMaterialFunctionCall_12, output=MA Out}]
   properties {attributeGetTypes [#2, #3, #4]}
  }
  MaterialExpressionGetMaterialAttributes_2 {class=MaterialExpressionGetMaterialAttributes, inputs [None {from=MaterialExpressionReroute_7, output=InputPin}]}
  MaterialExpressionGetMaterialAttributes_3 {
   class=MaterialExpressionGetMaterialAttributes
   inputs [None {from=MaterialExpressionMaterialFunctionCall_4, output=Result}]
   properties {attributeGetTypes [#5]}
  }
  MaterialExpressionLinearInterpolate_0 {
   class=MaterialExpressionLinearInterpolate
   inputs [
    A {from=MaterialExpressionMultiply_1}
    B {from=MaterialExpressionReroute_5}
    Alpha {from=MaterialExpressionSaturate_0}
   ]
   properties {desc="Fixes normals\r\non backfaces"}
  }
  MaterialExpressionLinearInterpolate_1 {
   class=MaterialExpressionLinearInterpolate
   inputs [
    A {from=MaterialExpressionConvert_1}
    B {from=MaterialExpressionMaterialFunctionCall_6, output=Result}
    Alpha {from=MaterialExpressionReroute_4}
   ]
  }
  MaterialExpressionLinearInterpolate_2 {
   class=MaterialExpressionLinearInterpolate
   inputs [
    A {from=MaterialExpressionMultiply_7}
    B {from=MaterialExpressionMultiply_19}
    Alpha {from=MaterialExpressionSaturate_2}
   ]
  }
  MaterialExpressionLinearInterpolate_3 {
   class=MaterialExpressionLinearInterpolate
   inputs [
    A {from=MaterialExpressionAdd_4}
    B {from=""}
    Alpha {from=MaterialExpressionScalarParameter_0}
   ]
  }
  MaterialExpressionLinearInterpolate_4 {
   class=MaterialExpressionLinearInterpolate
   inputs [
    A {from=MaterialExpressionReroute_9}
    B {from=MaterialExpressionConstant_0}
    Alpha {from=MaterialExpressionScalarParameter_1}
   ]
  }
  MaterialExpressionLinearInterpolate_5 {
   class=MaterialExpressionLinearInterpolate
   inputs [
    A {from=MaterialExpressionReroute_9}
    B {from=""}
    Alpha {from=MaterialExpressionScalarParameter_6}
   ]
   properties {constB=0}
  }
  MaterialExpressionLinearInterpolate_6 {
   class=MaterialExpressionLinearInterpolate
   inputs [
    A {from=""}
    B {from=""}
    Alpha {from=MaterialExpressionNamedRerouteUsage_8}
   ]
   properties {constA=1, constB=12}
  }
  MaterialExpressionLinearInterpolate_7 {
   class=MaterialExpressionLinearInterpolate
   inputs [
    A {from=MaterialExpressionTextureSample_1, output=RGB}
    B {from=MaterialExpressionLinearInterpolate_8}
    Alpha {from=MaterialExpressionScalarParameter_7}
   ]
  }
  MaterialExpressionLinearInterpolate_8 {
   class=MaterialExpressionLinearInterpolate
   inputs [
    A {from=MaterialExpressionTextureSample_1, output=G}
    B {from=MaterialExpressionMultiply_15}
    Alpha {from=MaterialExpressionSaturate_1}
   ]
  }
  MaterialExpressionMaterialFunctionCall_0 {
   class=MaterialExpressionMaterialFunctionCall
   inputs [
    AO {from=""}
    RVT_Height {from=MaterialExpressionReroute_1, output="RVT_Height (S)"}
    For Buildings {from=MaterialExpressionReroute_8, output="For Moss (MA)"}
    For Moss {from=""}
   ]
   properties {materialFunction=/Game/Environments/Asteria/Rocks/Materials/MaterialFunctions/MF_JungleBiome}
  }
  MaterialExpressionMaterialFunctionCall_1 {
   class=MaterialExpressionMaterialFunctionCall
   inputs [
    In {from=MaterialExpressionMaterialFunctionCall_10, output=Result}
    Custom Mask {from=""}
    Custom Rotation Axis {from=""}
   ]
   properties {materialFunction=/Game/Athena/WinterArt/Materials/MF_WindWPO_v2}
  }
  MaterialExpressionMaterialFunctionCall_10 {
   class=MaterialExpressionMaterialFunctionCall
   inputs [
    In {from=MaterialExpressionMaterialFunctionCall_11, output=Result}
    UseSphericalMappingOverride {from=""}
   ]
   properties {materialFunction=/Game/Packages/Fortress_SharedMaterials/Base_Material_Functions/MF_CurieFallbackEffect}
  }
  MaterialExpressionMaterialFunctionCall_11 {
   class=MaterialExpressionMaterialFunctionCall
   inputs [Material Attributes {from=MaterialExpressionMaterialFunctionCall_14, output=Result}]
   properties {materialFunction=/Game/Creative/Sets/MilitaryBase/Props/MF_FanSpin}
  }
  MaterialExpressionMaterialFunctionCall_12 {
   class=MaterialExpressionMaterialFunctionCall
   inputs [MA In {from=MaterialExpressionReroute_6, output=Output}]
   properties {materialFunction=&MF_CreativeMoveTool}
  }
  MaterialExpressionMaterialFunctionCall_13 {
   class=MaterialExpressionMaterialFunctionCall
   inputs [
    UVs {from=MaterialExpressionTextureCoordinate_0}
    Rotation Center {from=""}
    "Rotation Angle (0-1)" {from=MaterialExpressionConstant_3}
   ]
   properties {materialFunction=&CustomRotator}
  }
  MaterialExpressionMaterialFunctionCall_14 {
   class=MaterialExpressionMaterialFunctionCall
   inputs [In {from=MaterialExpressionMaterialFunctionCall_15, output=Material_Out}]
   properties {materialFunction=/Game/Athena/WinterArt/Materials/MF_WPO_WindNew}
  }
  MaterialExpressionMaterialFunctionCall_15 {
   class=MaterialExpressionMaterialFunctionCall
   inputs [
    Material_In {from=MaterialExpressionReroute_69, output=MaterialAttributes}
    UseDestructibleDamage {from=""}
    --- usage info tooltip --- {from=""}
    UseCrackedDamageTextures {from=""}
    UseDamageWobbleWPO {from=""}
    UseAdditiveWPO {from=""}
    WorldPosition {from=""}
    UseDestructionMultiplier {from=""}
    ExposeLerpFunctionForInvertingTheHitDistanceCalc {from=""}
    InvertDamageFalloff {from=""}
    UseAlternateHitGradientOutput {from=""}
    DestructibleDamageAmount {from=""}
    WSDeformationCenter {from=""}
    WSDamageRadius {from=""}
   ]
   properties {materialFunction=&MF_FN_DestructibleDamage}
  }
  MaterialExpressionMaterialFunctionCall_16 {
   class=MaterialExpressionMaterialFunctionCall
   inputs []
   properties {materialFunction=&/MF_GigaMix_Custom}
  }
  MaterialExpressionMaterialFunctionCall_17 {
   class=MaterialExpressionMaterialFunctionCall
   inputs [
    In {from=MaterialExpressionReroute_0, output=Result}
    "RVT Height Map (Roughness)" {from=MaterialExpressionGetMaterialAttributes_3, output=Roughness}
   ]
   properties {materialFunction=/Game/Global/Base4Layers_2022/MaterialFunctions/MF_RVTBlending}
  }
  MaterialExpressionMaterialFunctionCall_2 {
   class=MaterialExpressionMaterialFunctionCall
   inputs [In {from=MaterialExpressionReroute_2, output=Output}]
   properties {materialFunction=/Game/ArtTools/Tech_Art/DirectFX/Assets/MaterialFunctions/MF_DirectFX}
  }
  MaterialExpressionMaterialFunctionCall_3 {
   class=MaterialExpressionMaterialFunctionCall
   inputs [In {from=MaterialExpressionMaterialFunctionCall_1, output=Result}]
   properties {materialFunction=/Game/Global/Base4Layers_2022/MaterialFunctions/MF_SnowAndIceArchitecture}
  }
  MaterialExpressionMaterialFunctionCall_30 {
   class=MaterialExpressionMaterialFunctionCall
   inputs []
   properties {materialFunction=/Game/Global/Base4Layers_2022/MaterialFunctions/MF_GigaMix_2023}
  }
  MaterialExpressionMaterialFunctionCall_4 {
   class=MaterialExpressionMaterialFunctionCall
   inputs [In {from=MaterialExpressionMaterialFunctionCall_2, output=Result}]
   properties {materialFunction=/Game/ArtTools/Tech_Art/Veneer/MaterialFunctions/MF_Veneer}
  }
  MaterialExpressionMaterialFunctionCall_5 {
   class=MaterialExpressionMaterialFunctionCall
   inputs [In {from=MaterialExpressionMaterialFunctionCall_3, output=Result}]
   properties {materialFunction=/Game/Global/Base4Layers_2022/MaterialFunctions/MF_VibrationRattle}
  }
  MaterialExpressionMaterialFunctionCall_6 {
   class=MaterialExpressionMaterialFunctionCall
   inputs [
    Hue Shift Percentage {from=MaterialExpressionMultiply_2}
    Texture {from=MaterialExpressionConvert_1}
   ]
   properties {materialFunction=&HueShift}
  }
  MaterialExpressionMaterialFunctionCall_7 {
   class=MaterialExpressionMaterialFunctionCall
   inputs [In {from=MaterialExpressionNamedRerouteDeclaration_3}]
   properties {materialFunction=/Game/Effects/Fort_Effects/Materials/_Functions/MF_Smoothstep_01}
  }
  MaterialExpressionMaterialFunctionCall_8 {
   class=MaterialExpressionMaterialFunctionCall
   inputs [
    Number {from=MaterialExpressionNamedRerouteUsage_5}
    MaximumNumberOfDigits {from=MaterialExpressionConstant_1}
    UVs {from=MaterialExpressionMaterialFunctionCall_9, output=UVs}
    DebugTextLocation RG_UpperRight BA_LowerLeft {from=""}
   ]
   properties {materialFunction=&DebugScalarValues}
  }
  MaterialExpressionMaterialFunctionCall_9 {
   class=MaterialExpressionMaterialFunctionCall
   inputs [
    UVs {from=MaterialExpressionAdd_5}
    Texture Scale {from=MaterialExpressionConstant2Vector_0}
   ]
   properties {materialFunction=&ScaleUVsByCenter}
  }
  MaterialExpressionMaterialProxyReplace_0 {
   class=MaterialExpressionMaterialProxyReplace
   inputs [
    Realtime {from=MaterialExpressionLinearInterpolate_0, output=""}
    MaterialProxy {from=MaterialExpressionGetMaterialAttributes_0, output=Normal}
   ]
   properties {}
  }
  MaterialExpressionMax_3 {
   class=MaterialExpressionMax
   inputs [
    A {from=MaterialExpressionMultiply_148}
    B {from=MaterialExpressionMultiply_149}
   ]
  }
  MaterialExpressionMax_4 {
   class=MaterialExpressionMax
   inputs [
    A {from=MaterialExpressionMultiply_150}
    B {from=MaterialExpressionMultiply_151}
   ]
  }
  MaterialExpressionMax_5 {
   class=MaterialExpressionMax
   inputs [
    A {from=MaterialExpressionMax_3}
    B {from=MaterialExpressionMax_4}
   ]
  }
  MaterialExpressionMultiply_0 {
   class=MaterialExpressionMultiply
   inputs [
    A {from=MaterialExpressionScalarParameter_12}
    B {from=""}
   ]
  }
  MaterialExpressionMultiply_1 {
   class=MaterialExpressionMultiply
   inputs [
    A {from=MaterialExpressionReroute_5}
    B {from=""}
   ]
   properties {constB=-1}
  }
  MaterialExpressionMultiply_10 {
   class=MaterialExpressionMultiply
   inputs [
    A {from=MaterialExpressionAdd_1}
    B {from=MaterialExpressionReroute_10}
   ]
  }
  MaterialExpressionMultiply_11 {
   class=MaterialExpressionMultiply
   inputs [
    A {from=MaterialExpressionMultiply_4}
    B {from=MaterialExpressionScalarParameter_4}
   ]
  }
  MaterialExpressionMultiply_12 {
   class=MaterialExpressionMultiply
   inputs [
    A {from=MaterialExpressionSaturate_6}
    B {from=MaterialExpressionSaturate_7}
   ]
  }
  MaterialExpressionMultiply_13 {
   class=MaterialExpressionMultiply
   inputs [
    A {from=MaterialExpressionLinearInterpolate_5}
    B {from=""}
   ]
   properties {constB=0.5}
  }
  MaterialExpressionMultiply_14 {
   class=MaterialExpressionMultiply
   inputs [
    A {from=MaterialExpressionScalarParameter_9}
    B {from=MaterialExpressionNamedRerouteUsage_6}
   ]
  }
  MaterialExpressionMultiply_148 {
   class=MaterialExpressionMultiply
   inputs [
    A {from=MaterialExpressionOneMinus_15}
    B {from=MaterialExpressionScalarParameter_76}
   ]
  }
  MaterialExpressionMultiply_149 {
   class=MaterialExpressionMultiply
   inputs [
    A {from=MaterialExpressionOneMinus_20}
    B {from=MaterialExpressionScalarParameter_77}
   ]
  }
  MaterialExpressionMultiply_15 {
   class=MaterialExpressionMultiply
   inputs [
    A {from=MaterialExpressionTextureSample_1, output=R}
    B {from=MaterialExpressionNamedRerouteUsage_4}
   ]
  }
  MaterialExpressionMultiply_150 {
   class=MaterialExpressionMultiply
   inputs [
    A {from=MaterialExpressionOneMinus_21}
    B {from=MaterialExpressionScalarParameter_78}
   ]
  }
  MaterialExpressionMultiply_151 {
   class=MaterialExpressionMultiply
   inputs [
    A {from=MaterialExpressionOneMinus_22}
    B {from=MaterialExpressionScalarParameter_79}
   ]
  }
  MaterialExpressionMultiply_16 {
   class=MaterialExpressionMultiply
   inputs [
    A {from=MaterialExpressionMultiply_3}
    B {from=MaterialExpressionOneMinus_1}
   ]
  }
  MaterialExpressionMultiply_17 {
   class=MaterialExpressionMultiply
   inputs [
    A {from=MaterialExpressionScalarParameter_13}
    B {from=MaterialExpressionNamedRerouteUsage_7}
   ]
  }
  MaterialExpressionMultiply_18 {
   class=MaterialExpressionMultiply
   inputs [
    A {from=MaterialExpressionSaturate_9}
    B {from=MaterialExpressionSaturate_10}
   ]
  }
  MaterialExpressionMultiply_19 {
   class=MaterialExpressionMultiply
   inputs [
    A {from=MaterialExpressionMultiply_20}
    B {from=MaterialExpressionLinearInterpolate_6}
   ]
  }
  MaterialExpressionMultiply_2 {
   class=MaterialExpressionMultiply
   inputs [
    A {from=MaterialExpressionConstant_2}
    B {from=MaterialExpressionSine_1}
   ]
  }
  MaterialExpressionMultiply_20 {
   class=MaterialExpressionMultiply
   inputs [
    A {from=MaterialExpressionMultiply_7}
    B {from=MaterialExpressionNamedRerouteUsage_10}
   ]
  }
  MaterialExpressionMultiply_21 {
   class=MaterialExpressionMultiply
   inputs [
    A {from=MaterialExpressionSaturate_11}
    B {from=MaterialExpressionNamedRerouteUsage_9}
   ]
  }
  MaterialExpressionMultiply_3 {
   class=MaterialExpressionMultiply
   inputs [
    A {from=MaterialExpressionSaturate_3}
    B {from=MaterialExpressionSaturate_4}
   ]
  }
  MaterialExpressionMultiply_4 {
   class=MaterialExpressionMultiply
   inputs [
    A {from=MaterialExpressionNamedRerouteUsage_1}
    B {from=MaterialExpressionMultiply_10}
   ]
  }
  MaterialExpressionMultiply_5 {
   class=MaterialExpressionMultiply
   inputs [
    A {from=MaterialExpressionAdd_3}
    B {from=""}
   ]
   properties {constB=0.5}
  }
  MaterialExpressionMultiply_6 {
   class=MaterialExpressionMultiply
   inputs [
    A {from=MaterialExpressionLinearInterpolate_3}
    B {from=MaterialExpressionMultiply_5}
   ]
  }
  MaterialExpressionMultiply_7 {
   class=MaterialExpressionMultiply
   inputs [
    A {from=MaterialExpressionSaturate_5}
    B {from=MaterialExpressionAdd_7}
   ]
  }
  MaterialExpressionMultiply_8 {
   class=MaterialExpressionMultiply
   inputs [
    A {from=MaterialExpressionConvert_0}
    B {from=""}
   ]
   properties {constB=-1}
  }
  MaterialExpressionMultiply_9 {
   class=MaterialExpressionMultiply
   inputs [
    A {from=MaterialExpressionConvert_0}
    B {from=MaterialExpressionScalarParameter_3}
   ]
  }
  MaterialExpressionNamedRerouteDeclaration_0 {
   class=MaterialExpressionNamedRerouteDeclaration
   inputs [None {from=MaterialExpressionLinearInterpolate_7}]
   properties {name=SFX Mask, nodeColor=<rgba 0 1 0.5547 1>}
  }
  MaterialExpressionNamedRerouteDeclaration_1 {
   class=MaterialExpressionNamedRerouteDeclaration
   inputs [None {from=MaterialExpressionGetMaterialAttributes_1, output=Base Color}]
   properties {name=Base Color, nodeColor=<rgba 1 0.2344 0 1>}
  }
  MaterialExpressionNamedRerouteDeclaration_2 {
   class=MaterialExpressionNamedRerouteDeclaration
   inputs [None {from=MaterialExpressionMax_5}]
   properties {name=Layers, nodeColor=<rgba 0.1484 0 1 1>}
  }
  MaterialExpressionNamedRerouteDeclaration_3 {
   class=MaterialExpressionNamedRerouteDeclaration
   inputs [None {from=MaterialExpressionMultiply_16}]
   properties {name=Timeline_InOut, nodeColor=<rgba 0.5234 1 0 1>}
  }
  MaterialExpressionNamedRerouteDeclaration_4 {
   class=MaterialExpressionNamedRerouteDeclaration
   inputs [None {from=MaterialExpressionMaterialFunctionCall_7, output=Result}]
   properties {name=Timeline_InOutSmooth, nodeColor=<rgba 1 0 0.8438 1>}
  }
  MaterialExpressionNamedRerouteDeclaration_5 {
   class=MaterialExpressionNamedRerouteDeclaration
   inputs [None {from=MaterialExpressionClamp_0}]
   properties {name=ResetTimer, nodeColor=<rgba 1 0.3516 0 1>}
  }
  MaterialExpressionNamedRerouteDeclaration_6 {
   class=MaterialExpressionNamedRerouteDeclaration
   inputs [None {from=MaterialExpressionMultiply_18}]
   properties {name=Timeline, nodeColor=<rgba 0.5703 1 0 1>}
  }
  MaterialExpressionNamedRerouteDeclaration_7 {
   class=MaterialExpressionNamedRerouteDeclaration
   inputs [None {from=MaterialExpressionGetMaterialAttributes_1, output=Emissive Color}]
   properties {name=Emissive OG, nodeColor=<rgba 0 0.5547 1 1>}
  }
  MaterialExpressionNamedRerouteUsage_0 {class=MaterialExpressionNamedRerouteUsage}
  MaterialExpressionNamedRerouteUsage_1 {class=MaterialExpressionNamedRerouteUsage}
  MaterialExpressionNamedRerouteUsage_10 {class=MaterialExpressionNamedRerouteUsage}
  MaterialExpressionNamedRerouteUsage_11 {class=MaterialExpressionNamedRerouteUsage}
  MaterialExpressionNamedRerouteUsage_12 {class=MaterialExpressionNamedRerouteUsage}
  MaterialExpressionNamedRerouteUsage_2 {class=MaterialExpressionNamedRerouteUsage}
  MaterialExpressionNamedRerouteUsage_3 {class=MaterialExpressionNamedRerouteUsage}
  MaterialExpressionNamedRerouteUsage_4 {class=MaterialExpressionNamedRerouteUsage}
  MaterialExpressionNamedRerouteUsage_5 {class=MaterialExpressionNamedRerouteUsage}
  MaterialExpressionNamedRerouteUsage_6 {class=MaterialExpressionNamedRerouteUsage}
  MaterialExpressionNamedRerouteUsage_7 {class=MaterialExpressionNamedRerouteUsage}
  MaterialExpressionNamedRerouteUsage_8 {class=MaterialExpressionNamedRerouteUsage}
  MaterialExpressionNamedRerouteUsage_9 {class=MaterialExpressionNamedRerouteUsage}
  MaterialExpressionOneMinus_0 {class=MaterialExpressionOneMinus, inputs [None {from=MaterialExpressionSaturate_8}]}
  MaterialExpressionOneMinus_1 {class=MaterialExpressionOneMinus, inputs [None {from=MaterialExpressionMultiply_12}]}
  MaterialExpressionOneMinus_15 {class=MaterialExpressionOneMinus, inputs [None {from=MaterialExpressionSaturate_56}]}
  MaterialExpressionOneMinus_20 {class=MaterialExpressionOneMinus, inputs [None {from=MaterialExpressionSaturate_57}]}
  MaterialExpressionOneMinus_21 {class=MaterialExpressionOneMinus, inputs [None {from=MaterialExpressionSaturate_58}]}
  MaterialExpressionOneMinus_22 {class=MaterialExpressionOneMinus, inputs [None {from=MaterialExpressionSaturate_59}]}
  MaterialExpressionPinBase_0 {
   class=MaterialExpressionPinBase
   inputs []
   properties {reroutePins [Input {expression=&self:MaterialExpressionReroute_3}]}
  }
  MaterialExpressionPinBase_1 {
   class=MaterialExpressionPinBase
   inputs [Output {from=MaterialExpressionSetMaterialAttributes_0, output=""}]
   properties {reroutePins [Output {expression=&self:MaterialExpressionReroute_2}]}
  }
  MaterialExpressionPinBase_14 {
   class=MaterialExpressionPinBase
   inputs []
   properties {reroutePins [
     "For Moss (MA)" {expression=&self:MaterialExpressionReroute_8}
     "RVT_Height (S)" {expression=&self:MaterialExpressionReroute_1}
    ]}
  }
  MaterialExpressionPinBase_15 {
   class=MaterialExpressionPinBase
   inputs [Result {from=MaterialExpressionMaterialFunctionCall_0, output=Buildings Jungl}]
   properties {reroutePins [Result {expression=&self:MaterialExpressionReroute_0}]}
  }
  MaterialExpressionPinBase_2 {
   class=MaterialExpressionPinBase
   inputs []
   properties {reroutePins [MaterialAttributes {expression=&self:MaterialExpressionReroute_70}]}
  }
  MaterialExpressionPinBase_3 {
   class=MaterialExpressionPinBase
   inputs [Output {from=MaterialExpressionSetMaterialAttributes_1, output=""}]
   properties {reroutePins [Output {expression=&self:MaterialExpressionReroute_71}]}
  }
  MaterialExpressionPinBase_4 {
   class=MaterialExpressionPinBase
   inputs []
   properties {reroutePins [MaterialAttributes {expression=&self:MaterialExpressionReroute_69}]}
  }
  MaterialExpressionPinBase_5 {
   class=MaterialExpressionPinBase
   inputs [Result {from=MaterialExpressionMaterialFunctionCall_5, output=Result}]
   properties {reroutePins [Result {expression=&self:MaterialExpressionReroute_68}]}
  }
  MaterialExpressionPinBase_8 {
   class=MaterialExpressionPinBase
   inputs []
   properties {reroutePins [InputPin {expression=&self:MaterialExpressionReroute_7}]}
  }
  MaterialExpressionPinBase_9 {
   class=MaterialExpressionPinBase
   inputs [Output {from=MaterialExpressionShadingPathSwitch_0, output=""}]
   properties {reroutePins [Output {expression=&self:MaterialExpressionReroute_6}]}
  }
  MaterialExpressionQualitySwitch_0 {
   class=MaterialExpressionQualitySwitch
   inputs [
    Default {from=MaterialExpressionReroute_7, output=InputPin}
    Low {from=MaterialExpressionSetMaterialAttributes_2, output=""}
    High {from=MaterialExpressionReroute_7, output=InputPin}
    Medium {from=MaterialExpressionReroute_7, output=InputPin}
    Epic {from=MaterialExpressionReroute_7, output=InputPin}
   ]
   properties {}
  }
  MaterialExpressionReroute_0 {class=MaterialExpressionReroute, inputs [None {from=MaterialExpressionMaterialFunctionCall_0, output=Buildings Jungl}]}
  MaterialExpressionReroute_1 {class=MaterialExpressionReroute, inputs [None {from=MaterialExpressionMaterialFunctionCall_17, output=RVT H}]}
  MaterialExpressionReroute_10 {class=MaterialExpressionReroute, inputs [None {from=MaterialExpressionLinearInterpolate_4}]}
  MaterialExpressionReroute_2 {class=MaterialExpressionReroute, inputs [None {from=MaterialExpressionSetMaterialAttributes_0}]}
  MaterialExpressionReroute_3 {class=MaterialExpressionReroute, inputs [None {from=MaterialExpressionReroute_68, output=Result}]}
  MaterialExpressionReroute_4 {class=MaterialExpressionReroute, inputs [None {from=MaterialExpressionScalarParameter_3}]}
  MaterialExpressionReroute_5 {class=MaterialExpressionReroute, inputs [None {from=MaterialExpressionTransform_0}]}
  MaterialExpressionReroute_6 {class=MaterialExpressionReroute, inputs [None {from=MaterialExpressionShadingPathSwitch_0}]}
  MaterialExpressionReroute_68 {class=MaterialExpressionReroute, inputs [None {from=MaterialExpressionMaterialFunctionCall_5, output=Result}]}
  MaterialExpressionReroute_69 {class=MaterialExpressionReroute, inputs [None {from=MaterialExpressionReroute_71, output=Output}]}
  MaterialExpressionReroute_7 {class=MaterialExpressionReroute, inputs [None {from=MaterialExpressionMaterialFunctionCall_17, output=Result}]}
  MaterialExpressionReroute_70 {class=MaterialExpressionReroute, inputs [None {from=MaterialExpressionMaterialFunctionCall_16, output=Result}]}
  MaterialExpressionReroute_71 {class=MaterialExpressionReroute, inputs [None {from=MaterialExpressionSetMaterialAttributes_1}]}
  MaterialExpressionReroute_8 {class=MaterialExpressionReroute, inputs [None {from=MaterialExpressionMaterialFunctionCall_4, output=Result}]}
  MaterialExpressionReroute_9 {class=MaterialExpressionReroute, inputs [None {from=MaterialExpressionConvert_0}]}
  MaterialExpressionRound_6 {class=MaterialExpressionRound, inputs [None {from=MaterialExpressionAbs_1, output=""}], properties {}}
  MaterialExpressionRound_7 {class=MaterialExpressionRound, inputs [None {from=MaterialExpressionAbs_5, output=""}], properties {}}
  MaterialExpressionRound_8 {class=MaterialExpressionRound, inputs [None {from=MaterialExpressionAbs_6, output=""}], properties {}}
  MaterialExpressionRound_9 {class=MaterialExpressionRound, inputs [None {from=MaterialExpressionAbs_7, output=""}], properties {}}
  MaterialExpressionSaturate_0 {class=MaterialExpressionSaturate, inputs [None {from=MaterialExpressionTwoSidedSign_0}]}
  MaterialExpressionSaturate_1 {class=MaterialExpressionSaturate, inputs [None {from=MaterialExpressionNamedRerouteUsage_4}]}
  MaterialExpressionSaturate_10 {class=MaterialExpressionSaturate, inputs [None {from=MaterialExpressionScalarParameter_15}]}
  MaterialExpressionSaturate_11 {class=MaterialExpressionSaturate, inputs [None {from=MaterialExpressionNamedRerouteUsage_11}]}
  MaterialExpressionSaturate_2 {class=MaterialExpressionSaturate, inputs [None {from=MaterialExpressionSign_0}]}
  MaterialExpressionSaturate_3 {class=MaterialExpressionSaturate, inputs [None {from=MaterialExpressionDivide_2}]}
  MaterialExpressionSaturate_4 {class=MaterialExpressionSaturate, inputs [None {from=MaterialExpressionScalarParameter_8}]}
  MaterialExpressionSaturate_5 {class=MaterialExpressionSaturate, inputs [None {from=MaterialExpressionDotProduct_2}]}
  MaterialExpressionSaturate_56 {class=MaterialExpressionSaturate, inputs [None {from=MaterialExpressionRound_6}]}
  MaterialExpressionSaturate_57 {class=MaterialExpressionSaturate, inputs [None {from=MaterialExpressionRound_7}]}
  MaterialExpressionSaturate_58 {class=MaterialExpressionSaturate, inputs [None {from=MaterialExpressionRound_8}]}
  MaterialExpressionSaturate_59 {class=MaterialExpressionSaturate, inputs [None {from=MaterialExpressionRound_9}]}
  MaterialExpressionSaturate_6 {class=MaterialExpressionSaturate, inputs [None {from=MaterialExpressionDivide_3}]}
  MaterialExpressionSaturate_7 {class=MaterialExpressionSaturate, inputs [None {from=MaterialExpressionScalarParameter_10}]}
  MaterialExpressionSaturate_8 {class=MaterialExpressionSaturate, inputs [None {from=MaterialExpressionCeil_0}]}
  MaterialExpressionSaturate_9 {class=MaterialExpressionSaturate, inputs [None {from=MaterialExpressionDivide_4}]}
  MaterialExpressionScalarParameter_0 {class=MaterialExpressionScalarParameter, properties {defaultValue=1, group=1Custom, parameterName=Color rotation}}
  MaterialExpressionScalarParameter_1 {class=MaterialExpressionScalarParameter, properties {defaultValue=1, group=1Custom, parameterName=Color rotation}}
  MaterialExpressionScalarParameter_10 {class=MaterialExpressionScalarParameter, properties {bUseCustomPrimitiveData=true, parameterName=CPD_Timestamp_Start, primitiveDataIndex=2}}
  MaterialExpressionScalarParameter_11 {class=MaterialExpressionScalarParameter, properties {defaultValue=1, group=1Custom, parameterName=Timestamp_DurationOut}}
  MaterialExpressionScalarParameter_12 {
   class=MaterialExpressionScalarParameter
   properties {
    desc=Documentation link
    group="0 (Expand)  Documentation (RMB - Copy display name) and Debug view"
    parameterName=https://confluence.it.epicgames.com/display/FORT/Fortnite+Base+Material+2022
   }
  }
  MaterialExpressionScalarParameter_13 {class=MaterialExpressionScalarParameter, properties {defaultValue=0.2, group=1Custom, parameterName=Timestamp_DurationIn}}
  MaterialExpressionScalarParameter_14 {class=MaterialExpressionScalarParameter, properties {bUseCustomPrimitiveData=true, parameterName=CPD_Timestamp_Reset, primitiveDataIndex=3}}
  MaterialExpressionScalarParameter_15 {class=MaterialExpressionScalarParameter, properties {group=1Custom, parameterName=GameTime}}
  MaterialExpressionScalarParameter_16 {class=MaterialExpressionScalarParameter, properties {defaultValue=5, group=1Custom, parameterName=Duration}}
  MaterialExpressionScalarParameter_17 {class=MaterialExpressionScalarParameter, properties {defaultValue=4, group=1Custom, parameterName=Rotation Speed}}
  MaterialExpressionScalarParameter_18 {class=MaterialExpressionScalarParameter, properties {group=1Custom, parameterName=Emissive Min Bias}}
  MaterialExpressionScalarParameter_2 {class=MaterialExpressionScalarParameter, properties {group=1Custom, parameterName=Reactive Emissive?}}
  MaterialExpressionScalarParameter_3 {class=MaterialExpressionScalarParameter, properties {defaultValue=1, group=1Custom, parameterName=Color rotation}}
  MaterialExpressionScalarParameter_4 {class=MaterialExpressionScalarParameter, properties {defaultValue=1, group=1Custom, parameterName=Emissive Intensity Custom}}
  MaterialExpressionScalarParameter_5 {class=MaterialExpressionScalarParameter, properties {defaultValue=1, group=1Custom, parameterName=Speed}}
  MaterialExpressionScalarParameter_6 {class=MaterialExpressionScalarParameter, properties {defaultValue=1, group=1Custom, parameterName=Color rotation}}
  MaterialExpressionScalarParameter_7 {class=MaterialExpressionScalarParameter, properties {group=1Custom, parameterName=Use 2 Layers?}}
  MaterialExpressionScalarParameter_76 {class=MaterialExpressionScalarParameter, properties {defaultValue=1, group=Projector, parameterName=Projector L1, sortPriority=-4}}
  MaterialExpressionScalarParameter_77 {class=MaterialExpressionScalarParameter, properties {group=Projector, parameterName=Projector L2, sortPriority=-3}}
  MaterialExpressionScalarParameter_78 {class=MaterialExpressionScalarParameter, properties {group=Projector, parameterName=Projector L3, sortPriority=-2}}
  MaterialExpressionScalarParameter_79 {class=MaterialExpressionScalarParameter, properties {group=Projector, parameterName=Projector L4, sortPriority=-1}}
  MaterialExpressionScalarParameter_8 {class=MaterialExpressionScalarParameter, properties {bUseCustomPrimitiveData=true, parameterName=CPD_Timestamp_Start, primitiveDataIndex=2}}
  MaterialExpressionScalarParameter_9 {class=MaterialExpressionScalarParameter, properties {defaultValue=0.2, group=1Custom, parameterName=Timestamp_DurationIn}}
  MaterialExpressionSetMaterialAttributes_0 {
   class=MaterialExpressionSetMaterialAttributes
   inputs [
    MaterialAttributes {from=MaterialExpressionGetMaterialAttributes_0, output=MaterialAttributes}
    Normal {from=MaterialExpressionStaticSwitchParameter_0, output=""}
   ]
   properties {attributeSetTypes [#1]}
  }
  MaterialExpressionSetMaterialAttributes_1 {
   class=MaterialExpressionSetMaterialAttributes
   inputs [
    MaterialAttributes {from=MaterialExpressionReroute_70, output=MaterialAttributes}
    World Position Offset {from=MaterialExpressionMultiply_0, output=""}
   ]
   properties {attributeSetTypes [#6]}
  }
  MaterialExpressionSetMaterialAttributes_2 {
   class=MaterialExpressionSetMaterialAttributes
   inputs [
    MaterialAttributes {from=MaterialExpressionGetMaterialAttributes_2, output=MaterialAttributes}
    Specular {from=MaterialExpressionConstant_7, output=""}
    Roughness {from=MaterialExpressionConstant_6, output=""}
   ]
   properties {attributeSetTypes [#7, #5]}
  }
  MaterialExpressionSetMaterialAttributes_3 {
   class=MaterialExpressionSetMaterialAttributes
   inputs [
    MaterialAttributes {from=MaterialExpressionGetMaterialAttributes_1, output=MaterialAttributes}
    Base Color {from=MaterialExpressionNamedRerouteUsage_12, output=""}
    Emissive Color {from=MaterialExpressionEyeAdaptationInverse_0, output=EyeAdaptationInverse}
    Roughness {from="", output=""}
   ]
   properties {attributeSetTypes [#2, #3, #5]}
  }
  MaterialExpressionSetMaterialAttributes_4 {
   class=MaterialExpressionSetMaterialAttributes
   inputs [
    MaterialAttributes {from="", output=""}
    Base Color {from=MaterialExpressionMaterialFunctionCall_8, output=Result}
   ]
   properties {attributeSetTypes [#2]}
  }
  MaterialExpressionShadingPathSwitch_0 {
   class=MaterialExpressionShadingPathSwitch
   inputs [
    Default {from=MaterialExpressionReroute_7, output=InputPin}
    Deferred {from="", output=""}
    Forward {from="", output=""}
    Mobile {from=MaterialExpressionQualitySwitch_0, output=""}
   ]
   properties {}
  }
  MaterialExpressionSign_0 {class=MaterialExpressionSign, inputs [None {from=MaterialExpressionScalarParameter_2, output=""}], properties {}}
  MaterialExpressionSine_0 {class=MaterialExpressionSine, inputs [None {from=MaterialExpressionAdd_2}]}
  MaterialExpressionSine_1 {class=MaterialExpressionSine, inputs [None {from=MaterialExpressionDivide_0}]}
  MaterialExpressionStaticSwitchParameter_0 {
   class=MaterialExpressionStaticSwitchParameter
   inputs [
    True {from=MaterialExpressionMaterialProxyReplace_0, output=""}
    False {from=MaterialExpressionLinearInterpolate_0, output=""}
   ]
   properties {parameterName=Force TangentSpaceNormalsFor HLOD ProxyReplace}
  }
  MaterialExpressionStaticSwitchParameter_1 {
   class=MaterialExpressionStaticSwitchParameter
   inputs [
    True {from=MaterialExpressionMultiply_21, output=""}
    False {from=MaterialExpressionLinearInterpolate_2, output=""}
   ]
   properties {parameterName=IS lcd screen?}
  }
  MaterialExpressionSubtract_1 {
   class=MaterialExpressionSubtract
   inputs [
    A {from=MaterialExpressionTime_2}
    B {from=MaterialExpressionScalarParameter_8}
   ]
  }
  MaterialExpressionSubtract_2 {
   class=MaterialExpressionSubtract
   inputs [
    A {from=MaterialExpressionTime_3}
    B {from=MaterialExpressionAdd_0}
   ]
  }
  MaterialExpressionSubtract_3 {
   class=MaterialExpressionSubtract
   inputs [
    A {from=MaterialExpressionTime_4}
    B {from=MaterialExpressionScalarParameter_15}
   ]
  }
  MaterialExpressionTextureCoordinate_0 {class=MaterialExpressionTextureCoordinate}
  MaterialExpressionTextureCoordinate_5 {class=MaterialExpressionTextureCoordinate, properties {coordinateIndex=1}}
  MaterialExpressionTextureObjectParameter_30 {
   class=MaterialExpressionTextureObjectParameter
   inputs [
    UVs {from="", output=""}
    Apply View MipBias {from="", output=""}
   ]
   properties {
    desc="SFX UniversalTextureMask Uses Dot to select channel used for various parts of material.You can reuse that texture across layers and functions. "
    group=0_SFX Global
    parameterName=SFX_Mask
    samplerType=SAMPLERTYPE_Masks
    sortPriority=0
    texture=/Game/Global/Base4Layers_2022/Textures/T_SFX_M
   }
  }
  MaterialExpressionTextureSample_1 {
   class=MaterialExpressionTextureSample
   inputs [
    UVs {from=""}
    Tex {from=MaterialExpressionTextureObjectParameter_30}
    Apply View MipBias {from=""}
   ]
  }
  MaterialExpressionTime_0 {class=MaterialExpressionTime}
  MaterialExpressionTime_1 {class=MaterialExpressionTime}
  MaterialExpressionTime_2 {class=MaterialExpressionTime}
  MaterialExpressionTime_3 {class=MaterialExpressionTime}
  MaterialExpressionTime_4 {class=MaterialExpressionTime}
  MaterialExpressionTransform_0 {class=MaterialExpressionTransform, inputs [None {from=MaterialExpressionGetMaterialAttributes_0, output=Normal}]}
  MaterialExpressionTwoSidedSign_0 {class=MaterialExpressionTwoSidedSign, inputs [], properties {}}
 }
 properties {
  bAutomaticallySetUsageInEditor=false
  bTangentSpaceNormal=false
  bUseMaterialAttributes=true
  bUsedWithNanite=true
  ditheredLODTransition=true
  pixelDepthOffsetMode=PDOM_Legacy
  shadingModel=MSM_FromMaterialExpression
  usedShadingModels=Default Lit | Subsurface | Clear Coat | Two Sided Foliage
 }
}
